# Copyright Epic Games, Inc. All Rights Reserved.
#################################################
# Generated Digest of Verse API
# DO NOT modify this manually!
# Generated from build: ++Fortnite+Release-36.20-CL-43847582
#################################################


using {/Verse.org/SceneGraph}
batting_team_minigame_asset<scoped {/playkigai.dig@fortnite.com/Minigame_BattingTeam}>:entity_prefab = external {}

interaction_button_asset<scoped {/playkigai.dig@fortnite.com/Minigame_BattingTeam}>:entity_prefab = external {}

mechanical_batter_asset<scoped {/playkigai.dig@fortnite.com/Minigame_BattingTeam}>:entity_prefab = external {}

orbiting_ball_asset<scoped {/playkigai.dig@fortnite.com/Minigame_BattingTeam}>:entity_prefab = external {}

team_batting_setup_asset<scoped {/playkigai.dig@fortnite.com/Minigame_BattingTeam}>:entity_prefab = external {}