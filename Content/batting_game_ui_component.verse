<#
@UUID: 829d04c0-7441-42e4-beea-ba80e01218f6
@Task: Batting Team Game UI Component
@Description: Handles the pure-Verse UI of the Batting Teams Minigame. Reacts to game manager events to update the team's scores and display a timer and the game's result.
@Tags: UI, SceneGraph, DataStructures, Formatting
@Export: true
#>


<#
In-Editor Entity Prefab Setups:

interaction_button
    RootEntity
        ButtonBaseEntity
            - mesh_component
            - transform_component
                - Tweak the scale value so that it is large enough to envelop the ButtonHeadEntity
        ButtonHeadEntity
            - mesh_component (preferably a cylinder)
            - transform_component
                - Tweak the scale value so that the top pokes out of the top of ButtonBaseEntity
        InteractableColliderEntity 
            - interactable_component
            - mesh_component 
                - Collidable: unchecked
                - Visible: unchecked
            - transform_component:
                - Tweak the scale value so that it is just barely large enough to envelop all other mesh entities
        - Any other entities or components that are needed


team_batting_setup
    RootEntity
        InteractableButtonEntity (interaction_button Entity Prefab)
            - Tweak the transform value of this prefab's root entity so that the button is placed in front of the MechanicalBatterEntity's forwardVector
        MechanicalBatterEntity (mechanical_batter Entity Prefab)
        OrbitingBallEntity (mechanical_batter Entity Prefab)
            - Tweak the transform value of this prefab's root entity so that the ball is aligned with the left vector of the MechanicalBatterEntity
            - The MechanicalBatterEntity's bat collider must be able to collide with the ball when rotating at least 90 degrees clockwise from its starting rotaiton
        InteractableColliderEntity 
            - interactable_component
            - mesh_component 
                - Collidable: unchecked
                - Visible: unchecked
            - transform_component:
                - Tweak the scale value so that it is just barely large enough to envelop all other mesh entities
        - Any other entities or components that are needed


batting_team_minigame 
    RootEntity
        - batting_team_game_manager_component
            - Set any editable fields under GameTimer to your preference
            - Set "MaybeTeam1SetupRoot" to point to the root entity for "Team1Setup"
            - Set "MaybeTeam2SetupRoot" to point to the root entity for "Team2Setup"
        - batting_game_ui_component
            - Set any editable field to your preference
            - Ensure that this and the batting_team_game_manager are sibling components on the same entity
        Team1Setup (team_batting_setup Entity Prefab
        Team2Setup (team_batting_setup Entity Prefab)
            - Tweak the transform value of this prefab's root entity so that all the sub components are mirrored vs. their Team1Setup counterparts
            - Ensure that the mechanical_batter on this setup cannot hit the ball from the Team1Setup
        - Any other entities or components that are needed
#>


using { /UnrealEngine.com/Temporary/UI }
using { /UnrealEngine.com/Temporary/SpatialMath }
using { /Fortnite.com/UI }
using { /Fortnite.com/Devices }
using { /Fortnite.com/Teams }
using { /Fortnite.com/Playspaces }
using { /Verse.org }    
using { /Verse.org/Colors }
using { /Verse.org/Native }
using { /Verse.org/SceneGraph }
using { /Verse.org/Simulation }


###################### Start UI Component ##########################################################################

batting_game_ui_component<public> := class<final_super>(component):

    var PlayerToUI:[player]batting_game_ui = map{} # Stores the UI object for each player
    var MaybeGameManager: ?batting_team_game_manager_component = false  # Reference to the sibling batting_team_game_manager_component 
    

    OnBeginSimulation<override>():void =
        (super:)OnBeginSimulation()

        # Get Reference to Sibling Game Manager Component and setup gameplay callbacks
        set MaybeGameManager = option. Entity.GetComponent[batting_team_game_manager_component] 

        if (Playspace := Entity.GetPlayspaceForEntity[]):
            Playspace.PlayerRemovedEvent().Subscribe(OnPlayerRemoved)
            Playspace.PlayerRemovedEvent().Subscribe(OnPlayerRemoved)
        

    OnSimulate<override>()<suspends>:void=
        if (GameManager := MaybeGameManager?):
            
            # Add UI to every player's screen when the initializer device initilaizes the GameManager Component
            GameManager.BeginPlayEvent.Await()
            AddUIAllPlayers()

            loop:
                # Waiting For Player Ready
                race:
                    GameManager.GameTimer.TimerStartEvent.Await()
                    loop:   # On Team 1 Ready
                        GameManager.Team1ReadyEvent.Await()
                        ShowTeam1Ready()
                        Sleep(0.0)
                    loop:   # On Team 2 Ready
                        GameManager.Team2ReadyEvent.Await()
                        ShowTeam2Ready()
                        Sleep(0.0)

                # On Countdown Begin 
                UpdateTeam1Score(0)
                UpdateTeam2Score(0)
                race:
                    GameManager.GameBeginEvent.Await()   # On Game Start (and On Countdown Ends)
                    ListenTimerUpdate()

                # Game Active
                Winner: batting_team_game_winner = race:
                    block:   # On Game End
                        GameManager.GameOverEvent.Await()
                    block:
                        ListenScoreUpdate()
                        batting_team_game_winner.Draw
                    block:   # Timer Update
                        ListenTimerUpdate()
                        batting_team_game_winner.Draw
                        
                        
                # Post Game UI
                UpdateTimerText(EndGameMessage())   # Done!   
                case (Winner):
                    batting_team_game_winner.Team1 => UpdateResultText(Team1WinMessage())
                    batting_team_game_winner.Team2 => UpdateResultText(Team2WinMessage())
                    batting_team_game_winner.Draw => UpdateResultText(DrawMessage())

                GameManager.PostGameDelayComplete.Await()
                ResetUI()
                Sleep(0.0)
    

    ListenScoreUpdate()<suspends>:void =
        if (GameManager := MaybeGameManager?):
            loop:   # On Score Update
                Result := GameManager.ScoreUpdatedEvent.Await()
                case (Result.TeamNumber):
                    1=> UpdateTeam1Score(Result.NewScore)
                    2=> UpdateTeam2Score(Result.NewScore)
                    _=> Print("batting_game_ui_component:\nERROR: Invalid Team Number; couldn't update score")
                Sleep(0.0)


    ListenTimerUpdate()<suspends>:void=
        if (GameManager := MaybeGameManager?):
            loop:
                NewTime := GameManager.GameTimer.TimerUpdateEvent.Await()
                UpdateTimerText(NewTime)
                Sleep(0.0)

                        
    OnEndSimulation<override>():void =
        RemoveUIAllPlayers()    # Remove all UI elements on end; Custom Verse UI does not get cleaned up automatically.

        
    ShowTeam1Ready() : void = 
        for (Key->PlayerScoreUI : PlayerToUI):
            PlayerScoreUI.Team1ScoreText.SetText(Team1ReadyMessage())


    ShowTeam2Ready() : void = 
        for (Key->PlayerScoreUI : PlayerToUI):
            PlayerScoreUI.Team2ScoreText.SetText(Team2ReadyMessage())


    UpdateTeam1Score(NewScore : int) : void =
        for (Key->PlayerScoreUI : PlayerToUI):
            PlayerScoreUI.Team1ScoreText.SetText(Team1ScoreMessage(NewScore))


    UpdateTeam2Score(NewScore : int) : void =
        for (Key->PlayerScoreUI : PlayerToUI):
            PlayerScoreUI.Team2ScoreText.SetText(Team2ScoreMessage(NewScore))


    UpdateTimerText(NewTime:float):void = 
        if(FormattedTime := Ceil[NewTime]):
            for (Key->PlayerScoreUI : PlayerToUI):
                PlayerScoreUI.TimerText.SetText(IntToMessage(FormattedTime))
    

    UpdateTimerText(Message:message):void = 
        for (Key->PlayerScoreUI : PlayerToUI):
            PlayerScoreUI.TimerText.SetText(Message)
    

    UpdateResultText(Message:message):void = 
        for (Key->PlayerScoreUI : PlayerToUI):
            PlayerScoreUI.ResultText.SetText(Message)

            
    ResetUI():void = 
        for (Key->PlayerScoreUI : PlayerToUI):
            PlayerScoreUI.Reset()


    AddUIAllPlayers():void =
        if(Playspace := Entity.GetPlayspaceForEntity[]):
            for (Player : Playspace.GetPlayers()) do Player.AddUI()


    RemoveUIAllPlayers():void=
        if(Playspace := Entity.GetPlayspaceForEntity[]):
            for(Player : Playspace.GetPlayers()):
                Player.RemoveUI()
            set PlayerToUI = map{}
            

    (Player : player).AddUI():void =
        if:
            PlayerUI := GetPlayerUI[Player]
            BattingGameUI := batting_game_ui{}
            set PlayerToUI[Player] = BattingGameUI
        then:
            BattingGameUI.CreateUI()
            PlayerUI.AddWidget(BattingGameUI.GetWidget())
        else:
            Print("batting_game_ui_component:\nERROR: Could not add UI to player's screen")
        

    (Player : player).RemoveUI():void =
        if:
            PlayerUI := GetPlayerUI[Player]
            BattingGameUI := PlayerToUI[Player]
        then:
            PlayerUI.RemoveWidget(BattingGameUI.GetWidget())
            set PlayerToUI = PlayerToUI.RemoveKey(Player)
    

    OnPlayerAdded(Player: player):void = Player.AddUI()
    OnPlayerRemoved(Player: player):void = Player.RemoveUI()

###################### End UI Component #############################################################################


###################### Start UI Display Messages ######################################################################

InstructionMessage<localizes>() : message = "2 Players Must Press a Button to Begin"
StartGameMessage<localizes>() : message = "Start!"
EndGameMessage<localizes>() : message = "Done!"
DefaultScoreMessage<localizes>() : message = "Waiting"
Team1ScoreMessage<localizes>(Score:int) : message= "Team 1: {Score}"  
Team2ScoreMessage<localizes>(Score:int) : message= "Team 2: {Score}"  
Team1WinMessage<localizes>() : message = "Team 1 Wins!"
Team2WinMessage<localizes>() : message = "Team 2 Wins!"
Team1ReadyMessage<localizes>() : message = "Team 1 Ready!"
Team2ReadyMessage<localizes>() : message = "Team 2 Ready!"
DrawMessage<localizes>() : message = "Draw!"

###################### End UI Display Messages ##########################################################################


###################### Start Player UI Class ############################################################################
                
batting_game_ui := class:

    Team1ScoreText : text_block = text_block{DefaultTextColor := NamedColors.White}
    Team2ScoreText : text_block = text_block{DefaultTextColor := NamedColors.White}
    TimerText : text_block = text_block{DefaultTextColor := NamedColors.White}
    ResultText : text_block = text_block{DefaultTextColor := NamedColors.White}
  
    EndGameEvent<public> : event() = event(){}

    var MainWidget : canvas = canvas{}

    GetWidget() : canvas = MainWidget
        
    Reset() : void = 
        TimerText.SetText(InstructionMessage())
        ResultText.SetText(EmptyMessage())
        Team1ScoreText.SetText(DefaultScoreMessage())
        Team2ScoreText.SetText(DefaultScoreMessage())

    CreateUI() : void =
        #Displays: Team 1's Score    Team 2's Score
        ScoreUI := overlay:
            Slots := array:
                overlay_slot:
                    Widget := color_block:
                        DefaultColor := NamedColors.DarkGreen
                        DefaultOpacity := 1.0
                        DefaultDesiredSize := vector2{X := 512.0, Y := 64.0}  
                overlay_slot:
                    Widget := stack_box{
                        Orientation := orientation.Horizontal,
                        Slots := array{
                            stack_box_slot{
                                Widget := Team1ScoreText,
                                HorizontalAlignment := horizontal_alignment.Left,
                                VerticalAlignment := vertical_alignment.Center
                            },
                            stack_box_slot{
                                Widget := color_block:
                                    DefaultColor := NamedColors.CornflowerBlue
                                    DefaultOpacity := 0.0
                                    DefaultDesiredSize := vector2{X := 128.0, Y := 64.0}
                            },
                            stack_box_slot{
                                Widget := Team2ScoreText,
                                HorizontalAlignment := horizontal_alignment.Right,
                                VerticalAlignment := vertical_alignment.Center
                            }
                        }
                    }
        
        #Adds a timer and winner/draw text underneath the score text above
        OverallUI := stack_box{
            Orientation := orientation.Vertical,
            Slots := array{
                stack_box_slot{
                    Widget := ScoreUI,
                    HorizontalAlignment := horizontal_alignment.Center,
                    VerticalAlignment := vertical_alignment.Center
                },
                stack_box_slot{
                    Widget := TimerText,
                    HorizontalAlignment := horizontal_alignment.Center,
                    VerticalAlignment := vertical_alignment.Center
                },
                stack_box_slot{
                    Widget := ResultText,
                    HorizontalAlignment := horizontal_alignment.Center,
                    VerticalAlignment := vertical_alignment.Center
                }
            }
        }

        set MainWidget = canvas{
            Slots := array{
                canvas_slot{
                    Widget := OverallUI,
                    Anchors := anchors{Minimum := vector2{X := 0.5, Y := 0.1}, Maximum := vector2{X := 0.5, Y := 0.1}},
                    Alignment := vector2{X := 0.5, Y := 0.5}
                }
            }
        }

        Reset() 