<#
@UUID: 829d04c0-7441-42e4-beea-ba80e01218f6
@Task: Batting Team Game Manager Component
@Description: Handles the top-level game logic of the Batting Teams Minigame. This component interfaces with all other entities and components, Manages player interaction, score tracking, and the game timer.
@Tags: Gameplay, TeamManagement, PlayerInteraction, TimerManagement, DataStructures, SceneGraph, Devices
@Export: true
#>

<#
In-Editor Entity Prefab Setups:

interaction_button
    RootEntity
        ButtonBaseEntity
            - mesh_component
            - transform_component
                - Tweak the scale value so that it is large enough to envelop the ButtonHeadEntity
        ButtonHeadEntity
            - mesh_component (preferably a cylinder)
            - transform_component
                - Tweak the scale value so that the top pokes out of the top of ButtonBaseEntity
        InteractableColliderEntity 
            - interactable_component
            - mesh_component 
                - Collidable: unchecked
                - Visible: unchecked
            - transform_component:
                - Tweak the scale value so that it is just barely large enough to envelop all other mesh entities
        - Any other entities or components that are needed


team_batting_setup
    RootEntity
        InteractableButtonEntity (interaction_button Entity Prefab)
            - Tweak the transform value of this prefab's root entity so that the button is placed in front of the MechanicalBatterEntity's forwardVector
        MechanicalBatterEntity (mechanical_batter Entity Prefab)
        OrbitingBallEntity (mechanical_batter Entity Prefab)
            - Tweak the transform value of this prefab's root entity so that the ball is aligned with the left vector of the MechanicalBatterEntity
            - The MechanicalBatterEntity's bat collider must be able to collide with the ball when rotating at least 90 degrees clockwise from its starting rotaiton
        InteractableColliderEntity 
            - interactable_component
            - mesh_component 
                - Collidable: unchecked
                - Visible: unchecked
            - transform_component:
                - Tweak the scale value so that it is just barely large enough to envelop all other mesh entities
        - Any other entities or components that are needed


batting_team_minigame 
    RootEntity
        - batting_team_game_manager_component
            - Set any editable fields under GameTimer to your preference
            - Set "MaybeTeam1SetupRoot" to point to the root entity for "Team1Setup"
            - Set "MaybeTeam2SetupRoot" to point to the root entity for "Team2Setup"
        - batting_game_ui_component
            - Set any editable field to your preference
            - Ensure that this and the batting_team_game_manager are sibling components on the same entity
        Team1Setup (team_batting_setup Entity Prefab
        Team2Setup (team_batting_setup Entity Prefab)
            - Tweak the transform value of this prefab's root entity so that all the sub components are mirrored vs. their Team1Setup counterparts
            - Ensure that the mechanical_batter on this setup cannot hit the ball from the Team1Setup
        - Any other entities or components that are needed
#>


using { /Verse.org }
using { /Verse.org/Concurrency }
using { /Verse.org/Simulation }
using { /Verse.org/Native }
using { /Verse.org/SceneGraph }
using { /Fortnite.com/Game }
using { /Fortnite.com/Teams }
using { /Fortnite.com/Playspaces }
using { /Fortnite.com/Devices }
using { /UnrealEngine.com/Temporary/SpatialMath }
using { /UnrealEngine.com/Temporary/Diagnostics }


###################### Start Game Manager Component ###################################################

score_updated_result<public> := struct:
    TeamNumber<public>: int
    NewScore<public>: int


batting_team_game_winner<public> := enum:
    Team1,
    Team2,
    Draw

DebugCategory<public><localizes>:message := "Debug Settings"
PrefabSetupCategory<public><localizes>:message := "Entity Prefab Setup"

batting_team_game_manager_component<public> := class<final_super>(component):
    
    # batting_game_timer object that controls the pre-game countdown and duration of the minigame
    @editable
    GameTimer: batting_game_timer = batting_game_timer{}

    # Time (in seconds) after the game ends that the UI will persist on player's screens before disappearing
    @editable_number(float):
        MinValue := option. 0.0
    PostGameDelayTime : float = 10.0

    # Entity References that must point to the root entity for each team_batting_setup prefab
    @editable:
        Categories := array{PrefabSetupCategory}
    MaybeTeam1SetupRoot: ?entity = false
    @editable:
        Categories := array{PrefabSetupCategory}
    MaybeTeam2SetupRoot: ?entity = false

    # Flags used to override the need for their respective team to interact to begin the game; game automatically starts OnBeginPlay if both are true
    @editable:
        Categories := array{DebugCategory}
    DebugOverrideTeam1: logic = false
    @editable:
        Categories := array{DebugCategory}
    DebugOverrideTeam2: logic = false

    # ============ INTERNAL VARS =======================================
    # team_info data structures that contains all information about each team, including the Team Number, Team object reference, current score, and all game logic components
    var MaybeTeam1: ?team_info = false
    var MaybeTeam2: ?team_info = false

    # The team object reference for the neutral team (usually team 3); Used to effectively clear a player's team after the minigame is complete (since you cannot do so normally with the fort_team_collection API)
    var MaybeNeutralTeam: ?team = false

    var Teams: []team = array{}
    
    # ============ EVENTS =======================================
    # Signals the first time each respective team interacts with one of the buttons for the round
    Team1ReadyEvent<public>: event() = event(){}
    Team2ReadyEvent<public>: event() = event(){}
    # Meant to be signaled by an external device's begin play to tell this component that the game session has started
    BeginPlayEvent<public>: event() = event(){}    
    # Signals when a team presses their interact button; passes Index (int) of team who interacted
    TeamInteractEvent<public>: event(int) = event(int){}
    # Signals when a team's score is updated, passes team index (int) and new score value within a score_updated_result
    ScoreUpdatedEvent<public>: event(score_updated_result) = event(score_updated_result){} 
    # Signals when a minigame match has begun
    GameBeginEvent<public>: event() = event(){}
    # Signals when a minigame match has completed, passes who (if any) won.
    GameOverEvent<public>: event(batting_team_game_winner) = event(batting_team_game_winner){} 
    # Signals after the post game delay and just as the batting team game manager is ready to start another game
    PostGameDelayComplete<public>: event() = event(){} 
    

    OnPlayerAdded(Player: player):void=
        if:
            Playspace := Entity.GetPlayspaceForEntity[]
            NeutralTeam := MaybeNeutralTeam?
            Playspace.GetTeamCollection().AddToTeam[Player,NeutralTeam]
        then {}
        else Print("batting_team_game_manager_component:\nERROR: Neutral Team not set up")


    OnSimulate<override>()<suspends>:void=
        BeginPlayEvent.Await()

        if (Playspace := Entity.GetPlayspaceForEntity[]):
 
            set Teams = Playspace.GetTeamCollection().GetTeams()
            # Initialize each Team's Information Struct
            if (SetupRoot := MaybeTeam1SetupRoot?):
                set MaybeTeam1 = option. FindSetupComponents[1,SetupRoot]
            else Print("batting_team_game_manager_component:\nERROR: Team 1 Setup Root Entity Not set")
            if (SetupRoot := MaybeTeam2SetupRoot?):
                set MaybeTeam2 = option. FindSetupComponents[2,SetupRoot]
            else Print("batting_team_game_manager_component:\nERROR: Team 2 Setup Root Entity Not set")
            set MaybeNeutralTeam = option. Teams[2]

            # Check if there is a 3rd team to use as a neutral team
            if (NeutralTeam := MaybeNeutralTeam?):
                ResetTeams()
                Playspace.PlayerAddedEvent().Subscribe(OnPlayerAdded)
            else Print("batting_team_game_manager_component:\nERROR: Neutral Team not set up")

            if (Team1 := MaybeTeam1?, Team2 := MaybeTeam2? ):
                TeamCollection :=  Playspace.GetTeamCollection()
                loop:
                    # Both Teams must first interact before then game can start
                    sync:
                        loop:                      
                            if (DebugOverrideTeam1?) then break 
                            Agent := Team1.Interactable.SucceededEvent.Await()
                            if:
                                not TeamCollection.IsOnTeam[Agent,Team2.Team]
                                not TeamCollection.IsOnTeam[Agent,Team1.Team]
                                TeamCollection.AddToTeam[Agent,Team1.Team]
                            then:
                                Team1ReadyEvent.Signal()
                                break
                            Sleep(0.0)
                        loop:
                            if (DebugOverrideTeam2?) then break 
                            Agent := Team2.Interactable.SucceededEvent.Await()
                            if:
                                not TeamCollection.IsOnTeam[Agent,Team1.Team]
                                not TeamCollection.IsOnTeam[Agent,Team2.Team]
                                TeamCollection.AddToTeam[Agent,Team2.Team]
                            then:
                                Team2ReadyEvent.Signal()
                                break
                            Sleep(0.0)
                    
                    set Team1.Score = 0
                    ScoreUpdatedEvent.Signal(score_updated_result{TeamNumber:= Team1.TeamNumber,NewScore := Team1.Score})
                    set Team2.Score = 0
                    ScoreUpdatedEvent.Signal(score_updated_result{TeamNumber:= Team2.TeamNumber,NewScore := Team2.Score})
                                
                    # Main Game Logic
                    race:
                        GameTimer.Start()
                        block:  
                            GameTimer.CountdownTimerElapsedEvent.Await()
                            GameBeginEvent.Signal()
                            race:         
                                GameTimer.GameTimerElapsedEvent.Await()   # Team Coroutines will continue to run if we don't cancel this rac
                                block: 
                                    TeamLogicCoroutine(Team1)
                                    Sleep(Inf)
                                block:  
                                    TeamLogicCoroutine(Team2)
                                    Sleep(Inf)   
                    
                    # After the game is over, determine the winner 
                    Winner: batting_team_game_winner = 
                        if. Team1.Score > Team2.Score then batting_team_game_winner.Team1
                        else if. Team2.Score > Team1.Score then batting_team_game_winner.Team2
                        else batting_team_game_winner.Draw
                    
                    GameOverEvent.Signal(Winner)
                    ResetComponents(Team1)
                    ResetComponents(Team2)
                    ResetTeams()

                    # Post Game Delay 
                    Sleep(PostGameDelayTime)
                    PostGameDelayComplete.Signal()

            
    TeamLogicCoroutine(TeamInfo: team_info)<suspends>:void =
        Print("batting_team_game_manager_component:\nTeam {TeamInfo.TeamNumber} Coroutine Created")
        race:
            loop:   # Player Interaction
                Agent := TeamInfo.Interactable.SucceededEvent.Await()
                TeamInteractEvent.Signal(TeamInfo.TeamNumber)
                if:
                    Playspace := Entity.GetPlayspaceForEntity[]
                    Playspace.GetTeamCollection().IsOnTeam[Agent, TeamInfo.Team]
                    GameTimer.TimerStage = batting_timer_stage.GameActive
                then:
                    race:
                        TeamInfo.Batter.SwingBat()
                        loop:
                            TeamInfo.Interactable.SucceededEvent.Await()
                            # Print("batting_team_game_manager_component:\nTeam {TeamInfo.TeamNumber}'s batter is in the middle of swinging, wait until it has reset!")
                            Sleep(0.0)
                Sleep(0.0)

            loop:   # Team Scoring
                TeamInfo.Ball.BallHitEvent.Await()
                set TeamInfo.Score += 1
                Print("batting_team_game_manager_component:\nTeam {TeamInfo.TeamNumber}'s Score is {TeamInfo.TeamNumber}")
                ScoreUpdatedEvent.Signal(score_updated_result{TeamNumber:= TeamInfo.TeamNumber,NewScore := TeamInfo.Score})
                Sleep(0.0)

            loop:
                Print("batting_team_game_manager_component:\nTeam {TeamInfo.TeamNumber} Coroutine Tick")
                Sleep(2.0)
    

    ResetComponents(TeamInfo: team_info):void=
        TeamInfo.Batter.ResetPostion()
        TeamInfo.Ball.ResetPostion()
    
    ResetTeams():void =
        if (Playspace := Entity.GetPlayspaceForEntity[], NeutralTeam := MaybeNeutralTeam?) :
            for:
                Player : Playspace.GetPlayers()
                Playspace.GetTeamCollection().AddToTeam[Player,NeutralTeam]
            do {}
        else Print("batting_team_game_manager_component:\nERROR: Could not reset players to neutral team")


    FindSetupComponents(TeamNumber: int, SetupRoot: entity)<decides><transacts>:team_info= 
        Team := Teams[TeamNumber-1]
        # Try to find the first component of each type on Team Batting Setup root entity
        # NOTE: the first element will not be consistent due to how component querying works. This code assumes the that the root only has one of each component both attached to and decending from it
        Batter := SetupRoot.GetDescendantComponentsArray(mechanical_batter_component)[0]
        Ball := SetupRoot.GetDescendantComponentsArray(orbiting_ball_component)[0]
        Interactable := SetupRoot.GetDescendantComponentsArray(interactable_component)[0]
        MakeTeamInfo(TeamNumber, Team, Batter, Ball, Interactable)

###################### End Game Manager Component ############################################################


###################### Start Game Timer ######################################################################

batting_timer_stage := enum {Standby,Countdown,GameActive}

# Class that handles the pregame countdown timer and the game timer. Fires custom events that allow other components to react to them with gameplay events
batting_game_timer := class:

    # ============ EDITABLE FIELDS =======================================

    # Time in seconds that the game will last
    @editable_slider(float):
        MinValue := option. 0.0     # Updates every frame
    GameDuration: float = 20.0

    # Time in seconds before the game actually starts
    @editable_slider(float):
        MinValue := option. 0.0     # Updates every frame
    CountdownDuration: float = 3.0

    # Rate in seconds that the timer is updated
    @editable_slider(float):
        MinValue := option. 0.0     # Updates every frame
        MaxValue := option. 1.0     # Updates once per second
    UpdateRate: float = 1.0
    
    # ============ INTERNAL VARS ====================================

    var TimerStage: batting_timer_stage = batting_timer_stage.Standby
    var TimeRemaining: float = 0.0

    # Setting this value to "true" will cause the timer to stop updating until it is "false" again
    var Paused<private>: logic = false

    # ============ EVENTS =======================================
    # Signals when the Start function is first called and when the countdown begins
    TimerStartEvent<public> : event() = event(){}
    # Signals when the countdown timer completes and when the game timer begins
    CountdownTimerElapsedEvent<public> : event() = event(){}
    # Signals when the game timer has elapsed but before a winner is determined by the game manager component
    GameTimerElapsedEvent<public> : event() = event(){}
    # Signals every time the TimeRemaining value is updated (countdown and game timer), passes the current remaining time
    TimerUpdateEvent<public> : event(float) = event(float){}
    # Signals when timer goes from playing to being paused.
    PauseEvent<public> : event() = event(){}
    # Signals when timer goes from being paused to being playing again.
    ResumeEvent<public> : event() = event(){}


    # Immediately begins an asyncronous countdown timer which leads into the normal game timer
    # Function completes when the Game Timer fully elapses
    # Spawning this funciton can allow the timer logic to continue to run past the original calling context
    Start<public>()<suspends> : void =
        # Prevent multiple instances of this function from running at the same time.
        if (not TimerStage = batting_timer_stage.Standby) then return

        # Begin the countdown timer
        set TimerStage = batting_timer_stage.Countdown
        SubTimer(CountdownDuration)
        TimerStartEvent.Signal()
        
        # On Countdown Elapsed, Begin Game Timer
        set TimerStage = batting_timer_stage.GameActive 
        SubTimer(GameDuration)
        CountdownTimerElapsedEvent.Signal()

        # On Game Timer Elapsed
        set TimerStage = batting_timer_stage.Standby
        GameTimerElapsedEvent.Signal()
    
    IsPaused<public>()<transacts><decides>:void= Paused? 

    Pause<public>():void=
        if (TimerStage = batting_timer_stage.Standby) then return
        set Paused = true
        PauseEvent.Signal()


    Resume<public>():void=
        if (TimerStage = batting_timer_stage.Standby) then return
        set Paused = false
        ResumeEvent.Signal()


    SubTimer<private>(Duration:float)<suspends>:void=
        set TimeRemaining = Duration
        TimerUpdateEvent.Signal(TimeRemaining)
        loop:
            if (Paused?):
                ResumeEvent.Await()
            # Calculate the delta between frames
            Timestamp := GetSimulationElapsedTime()
            Sleep(UpdateRate)
            DeltaTime := GetSimulationElapsedTime() - Timestamp

            set TimeRemaining -= DeltaTime  # Update the RemainingTime
            TimerUpdateEvent.Signal(TimeRemaining)

            if (TimeRemaining <= 0.0):
                set TimeRemaining = 0.0
                break

    
###################### End Game Timer #########################################################################


###################### Start Team Info Container ##############################################################

MakeTeamInfo<constructor>(TeamNumber : int, Team : team, Batter : mechanical_batter_component,Ball : orbiting_ball_component, Interactable : interactable_component)<transacts> := team_info:
    TeamNumber := TeamNumber
    Team := Team
    Batter := Batter
    Ball := Ball
    Interactable := Interactable

team_info<public> := class():
    TeamNumber<public> : int
    Team<public> : team
    Batter<public>: mechanical_batter_component
    Ball<public>: orbiting_ball_component
    Interactable<public>: interactable_component
    var Score: int = 0

###################### End Team Info Container ##############################################################


###################### Start Initializer Device #############################################################  

batting_team_initializer:= class(creative_device):
    OnBegin<override>()<suspends>:void=
        if (SimEntity := GetSimulationEntity[]):
            for:
                Entity : SimEntity.FindDescendantEntitiesWithComponent(batting_team_game_manager_component)
                Comp := Entity.GetComponent[batting_team_game_manager_component]
            do:
                InitializeComponent(Comp)  
    

    InitializeComponent(Comp:batting_team_game_manager_component):void=
        if (SimEntity := GetSimulationEntity[]):
            # Add to the simulation entity if not there already
            if (not Comp.IsInScene[]):
                SimEntity.AddEntities(array{Comp.Entity})    
            
            Comp.BeginPlayEvent.Signal()

###################### End Iniitializer Device #############################################################
