<#
@UUID: 829d04c0-7441-42e4-beea-ba80e01218f6
@Task: Mechanical Batter Component
@Description: Scene Graph Component that manages the game logic for the mechanical_batter entity prefab. This component has a public function which causes the batter to swing and reset to its starting position. It also handles hit detection with the orbiting_ball prefab.
@Tags: Gameplay, SpatialMath, CollisionDetection, SceneGraph
@Export: true
#>

<#
In-Editor Entity Prefab Setup:

mechanical_batter
    RootEntity
        - mechanical_batter_component
            - set all editable fields to your preference
        - transform_component
            - any transform_component settings
        BatColliderEntity
            - mesh_component (preferably a cylinder)
                - Enabled: Checked
                - Queryable: Checked
                - Collidable: Checked
        - Any other entities or components that are needed
#>

using { /Verse.org }
using { /Verse.org/Concurrency }
using { /Verse.org/Simulation }
using { /Verse.org/Native }
using { /Verse.org/SceneGraph }
using { /Verse.org/Simulation }
using { /Verse.org/SpatialMath }
using { /Fortnite.com/Game }
using { /Fortnite.com/Devices }
using { /UnrealEngine.com/Temporary/SpatialMath }
using { /UnrealEngine.com/Temporary/Diagnostics }


mechanical_batter_component<public> := class<final_super>(component):
    # The arc in degrees that the bat will swing from its start position
    @editable_slider(float):
        MinValue := option. 90.0
        MaxValue := option. 180.0
    SwingArc: float = 90.0

    # The speed in degrees per second that the bat swings
    @editable_number(float):
        MinValue := option. 0.0
    SwingSpeed: float = 10.0

    # The speed in degrees per second that the bat returns to its starting position after swinging
    @editable_number(float):
        MinValue := option. 0.0
    ResetSpeed: float = 10.0
    
    # Amount of time in seconds that the batter waits after it finishes swinging before it starts to reset
    @editable_number(float):
        MinValue := option. 0.0
    EndLag: float = 0.0

    # Amount of time in seconds between each of rotation movement tick, if 0.0, update every frame
    @editable_number(float):
        MinValue := option. 0.0
    RotationUpdateTick: float = 0.0

    # Amount of time in seconds between each collision check on the bat, if 0.0, update every frame
    @editable_number(float):
        MinValue := option. 0.0
    CollisionUpdateTick: float = 0.0

    var ResetRotation: (/Verse.org/SpatialMath:)rotation = (/Verse.org/SpatialMath:)IdentityRotation()
    var MaybeBatCollider: ?mesh_component = false

    SimulationEndEvent<public>:event() = event(){}

    
    OnBeginSimulation<override>():void=
        (super:)OnBeginSimulation()
        set ResetRotation = Entity.GetLocalTransform().Rotation
        # Get the bat collider
        Meshes := Entity.FindDescendantComponents(mesh_component)
        for:
            MeshComp : Meshes
            Entity <> MeshComp.Entity 
            MeshComp.Collidable?          
        do:
            set MaybeBatCollider = option. MeshComp
        if (not MaybeBatCollider?) then Print("mechanical_batter_component: Bat collider not found, exiting coliision logic")
        
    
    OnSimulate<override>()<suspends>:void=
        SwingBat()  # Swing the bat once when the simulation game starts to avoid a possible physics bug.


    OnEndSimulation<override>():void= 
        (super:)OnEndSimulation()
        SimulationEndEvent.Signal()


    SwingBat<public>()<suspends>:void=
        TargetRot := FromRotation(FromRotation(ResetRotation).ApplyYaw((/Verse.org/SpatialMath:)DegreesToRadians(SwingArc)))
        race:        # Swing the bat and run bat collision while its swinging
            RotateToTargetLocalRotation(TargetRot,SwingSpeed)
            block:
                BatCollisionTick()
                Sleep(Inf)
        Sleep(EndLag)   # Reset Delay
        race:           # Swing bat back to start position
            RotateToTargetLocalRotation(ResetRotation,ResetSpeed)
            block:
                BatCollisionTick(?InvertHitDirection := true)
                Sleep(Inf)
    

    RotateToTargetLocalRotation(TargetRotation : (/Verse.org/SpatialMath:) rotation, RotSpeed : float, ?ArrivalThreashold: float = 0.01)<suspends> : void =
        StartRotation := Entity.GetLocalTransform().Rotation
        loop:
            # Calculate the shortest rotation between current and target using AngularDistance
            PrevLocal := Entity.GetLocalTransform()
            RotationDelta := AngularDistance(FromRotation(PrevLocal.Rotation), FromRotation(TargetRotation))

            # Check if we're close enough to the target rotation
            if (RotationDelta < ArrivalThreashold):
                break

            # Calculate the delta between frames
            Timestamp := GetSimulationElapsedTime()
            Sleep(RotationUpdateTick)
            DeltaTime := GetSimulationElapsedTime() - Timestamp

            # Calculate the rotation step based on speed and delta time
            Interp := InterpParam := Clamp(RotSpeed*DeltaTime, 0.0, 1.0)
            NewRotation := Slerp(PrevLocal.Rotation, TargetRotation, Interp)
            NewLocal :=  (/Verse.org/SpatialMath:)transform{
                Translation := PrevLocal.Translation,
                Rotation := NewRotation,
                Scale := PrevLocal.Scale
            }
            
            # Apply new rotation
            Entity.SetLocalTransform(NewLocal)
        
        FinalLocal := (/Verse.org/SpatialMath:) transform{
            Translation := Entity.GetLocalTransform().Translation,
            Rotation := TargetRotation,
            Scale := Entity.GetLocalTransform().Scale
        }
        Entity.SetLocalTransform(FinalLocal)


    # Continuously look for collisions with orbiting ball components
    BatCollisionTick<protected>(?InvertHitDirection: logic = false)<suspends>:void=
        race:
            SimulationEndEvent.Await()  # Exits this just in case this function is spawned
            block:
                if (BatCollider := MaybeBatCollider?):
                    loop:
                        OverlappedEntity := BatCollider.EntityEnteredEvent.Await()
                        for:
                            BallComp : OverlappedEntity.FindAncestorComponents(orbiting_ball_component)
                        do:
                            BallComp.OnHit(?InvertHit := InvertHitDirection )
                            
                        Sleep(CollisionUpdateTick)
                else Print("mechanical_batter: WARNING Bat Collider NOT Found, Exiting Collision Logic")



    ResetPostion<public>():void = 
        OriginTransform := (/Verse.org/SpatialMath:) transform{
            Translation := Entity.GetLocalTransform().Translation,
            Rotation := ResetRotation,
            Scale := Entity.GetLocalTransform().Scale
        }
        Entity.SetLocalTransform(OriginTransform)
