<#
@UUID: 829d04c0-7441-42e4-beea-ba80e01218f6
@Task: Orbiting Ball Component
@Description: Component that manages the game logic for the orbiting_ball entity prefab. When the ball is struck by the mechanical_batter prefab, it will horizontally rotate about a center axis while gradually slowing down until reaching a certain speed threashhold and then resetting back to its starting position.
@Tags: Gameplay, Physics, SpatialMath, SceneGraph
@Export: true
#>

<#
In-Editor Entity Prefab Setup:

orbiting_ball
    RootEntity
        - orbiting_ball_component
            - set all editable fields to your preference
        BodyEntity
            - - mesh_component 
                - Collidable: Unchecked
        BallEntity
            - mesh_component (preferably a Sphere)
                - Enabled: Checked
                - Queryable: Checked
                - Collidable: Checked
        - Any other entities or components that are needed
#>

using { /Verse.org }
using { /Verse.org/Concurrency }
using { /Verse.org/Simulation }
using { /Verse.org/Native }
using { /Verse.org/SceneGraph }
using { /Verse.org/SpatialMath }
using { /Verse.org/Simulation }
using { /Fortnite.com/Game }
using { /Fortnite.com/Devices }
using { /UnrealEngine.com/Temporary/SpatialMath }
using { /UnrealEngine.com/Temporary/Diagnostics }

###################### Start Orbiting Ball Component ###################################################

orbiting_ball_component<public> := class<final_super>(component):

    # The degrees per second that the ball slows after being hit
    @editable_number(float):
        MinValue := option. 0.0
    Friction: float = 2.5
    
    # The speed in degrees/s that the ball starts at when hit
    @editable_number(float):
        MinValue := option. 0.0
    InitialSpeed: float = 14.0

    # The minimum speed in degrees per second the ball can decay to before it returns to its start position
    @editable_number(float):
        MinValue := option. 0.0
    MinSpeed: float = 1.5

    # Amount of time in seconds between each of rotation movement update, if 0.0, update every frame
    @editable_number(float):
        MinValue := option. 0.0
    RotationUpdateTick: float = 0.0

    # Angular Distance (in Degrees) that the ball must be within it origin rotation to stop after slowing to the min speed
    @editable_number(float):
        MinValue := option. 0.0
    ArrivalThreashold: float = 0.6
    
    # if checked, will invert the default rotation direction of the ball. 
    @editable
    InvertDirection: logic = false

    # Axis (up vector for horizontal rotation)
    RotationAxis : (/Verse.org/SpatialMath:)vector3 = (/Verse.org/SpatialMath:)vector3{Forward := 0.0, Left := 0.0, Up := 1.0}
    
    var CurrentSpeed: float = 0.0
    var ResetRotation: (/Verse.org/SpatialMath:)rotation = (/Verse.org/SpatialMath:)IdentityRotation()
    
    BallHitEvent<public>:event() = event(){}    # used by other components and devices to track scoring
    
    
    OnBeginSimulation<override>():void=
        (super:)OnBeginSimulation()
        set ResetRotation = Entity.GetLocalTransform().Rotation


    OnSimulate<override>()<suspends>:void=
        loop:
            PrevLocal := Entity.GetLocalTransform()

            # Calculate accurate time delta 
            Timestamp := GetSimulationElapsedTime()
            Sleep(RotationUpdateTick)
            DeltaTime := GetSimulationElapsedTime() - Timestamp
            
            # Spinning code will begin to run regardless of where the CurrentSpeed is set
            if (CurrentSpeed <> 0.0):
                RotationDelta := MakeRotation(FromVector3(RotationAxis), CurrentSpeed*DeltaTime)
                NewLocalRot := FromRotation(FromRotation(PrevLocal.Rotation).RotateBy(RotationDelta))

                NewLocal := (/Verse.org/SpatialMath:)transform{
                    Translation := PrevLocal.Translation,
                    Rotation := NewLocalRot,
                    Scale := PrevLocal.Scale
                }
                
                # Apply new rotation
                Entity.SetLocalTransform(NewLocal)

                # Apply friction to the rotation speed until it reaches the minimum
                if (Abs(CurrentSpeed) > MinSpeed):
                    set CurrentSpeed = Max(Abs(CurrentSpeed) - (Friction*DeltaTime) , MinSpeed)  * Sgn(CurrentSpeed)
                else:
                    set CurrentSpeed = MinSpeed * Sgn(CurrentSpeed)
                    AngularDistanceToReset := AngularDistance(FromRotation(PrevLocal.Rotation), FromRotation(ResetRotation))
                    if (AngularDistanceToReset < (/Verse.org/SpatialMath:)DegreesToRadians(ArrivalThreashold)):
                        set CurrentSpeed = 0.0
    
    
    OnHit<public>(?InvertHit: logic = false ):void= 
        set CurrentSpeed = InitialSpeed * InvertDirection.ToFloatSign() * (-1*InvertHit.ToFloatSign())
        BallHitEvent.Signal()
    

    ResetPostion<public>():void =
        set CurrentSpeed = 0.0
        OriginTransform := (/Verse.org/SpatialMath:) transform{
            Translation := Entity.GetLocalTransform().Translation,
            Rotation := ResetRotation,
            Scale := Entity.GetLocalTransform().Scale
        }
        Entity.SetLocalTransform(OriginTransform)

###################### End Orbiting Ball Component ###################################################

