<#
@UUID: 829d04c0-7441-42e4-beea-ba80e01218f6
@Task: Batting Team Minigame.
@Description: Create a scene-graph based, two-player minigame where each "team" uses a batting mechanism to hit an orbiting baseball more than the opposing tean within a given time limit. 
    Before the game begins, all players recieve a UI element on their screen that instructs each person to press a button. Once both teams have pressed their button,
    A countodwn timer begins ticking down, and the game officially begins once it reaches 0. During the game, pressing the interact button causes a mechanical batter 
    to swing before quickly reseting to its starting position. While the bat is swining in either direction, it detects for collisions between its 
    bat collider entity and a separate orbiting ball entity. Collison between the two causes the ball to quickly orbit around a centerpoint in the opposite direction of the bat's velocity,
    The ball's rotation speed should gradually slowing until it reaches a minimum speed where it will then continue to orbit until it reaches its starting positon in fron of the bat.
    When a team successfully lands a hit on their ball, they should recieve a point and the current score of both teams should be dispalyed at the top of the screen.
    A game timer shouold tick down on each player's screen while they play, and the team who has the highest score by the time it expires is declared the winner.
    After a brief time, the Bat and ball setups should reset to their starting positions, and players should be able to begin a new game by pressing the buttons again.
    Each game component should be as configurable as possible and should be able divided into modular Entity Prefabs.
@Tags: Minigame, Sports, Arcade, Physics, PvP, NonCombat
@Export: true
#>

