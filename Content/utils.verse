<#
@UUID: 829d04c0-7441-42e4-beea-ba80e01218f6
@Task: Batting Team Game Utility Functions
@Description: Includes generic helper functions for various tasks.
@Tags: Utility, Formatting, DataConversion, DataStructures, SceneGraph
@Export: true
#>

using { /Verse.org/SceneGraph }

(Val:logic).ToIntSign():int = if. Val? then 1 else -1
(Val:logic).ToFloatSign():float = if. Val? then 1.0 else -1.0

EmptyMessage<localizes><public>() : message = ""

FloatToMessage<localizes><public>(Float:float):message = "{Float}"
IntToMessage<localizes><public>(Num:int):message = "{Num}"

# Conatenate 2 messages
operator'+'<localizes><public>(M1:message, M2:message)<transacts>:message="{M1}{M2}"

# Removes an element from the given map and returns a new map without that element
(Map:[keytype]valuetype where keytype:subtype(comparable), valuetype:type).RemoveKey(ElementToRemove:keytype):[keytype]valuetype=
    var NewMap:[keytype]valuetype = map{}
    for (Key -> Value : Map, Key <> ElementToRemove):
        set NewMap = ConcatenateMaps(NewMap, map{Key => Value})
    return NewMap

# Helper function to return an array (instead of a generator) of all components attached to descendant entities to and including `InEntity` of type `component_type`.
(Entity:entity).GetDescendantComponentsArray<public>(ComponentType:castable_subtype(component))<transacts>:[]ComponentType =
    var Components:[]ComponentType = array{}
    for (Comp : Entity.FindDescendantComponents(ComponentType)):
        set Components += array{Comp}
    Components